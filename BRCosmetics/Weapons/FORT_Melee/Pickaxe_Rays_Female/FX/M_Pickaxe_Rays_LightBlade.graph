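# Translucent unlit material for the light blade of the Rays Female pickaxe, with an animated energy glow and reveal.
# Flowmapped plasma noise tints the powered diffuse with EnergyColor, then gains a normal-based directional highlight and pivot-centred internal light to form emissive.
# Opacity is a smoothstep driven by Reveal parameters, with a timing map, positional bias and optional looping debug timer.
# A flat UV set is projected from pre-skinned XZ position.
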Material /BRCosmetics/Weapons/FORT_Melee/Pickaxe_Rays_Female/FX/M_Pickaxe_Rays_LightBlade
graph {
 properties {
  allowTranslucentCustomDepthWrites=true
  bCastRayTracedShadows=false
  bEnableResponsiveAA=true
  bUseMaterialAttributes=true
  bUsedWithMeshDeformer=true
  bUsedWithSkeletalMesh=true
  blendMode=BLEND_Translucent
  opacityMaskClipValue=0
  pixelDepthOffsetMode=PDOM_Legacy
  shadingModel=MSM_Unlit
  usedShadingModels=Unlit
 }
 nodes [
  n1 = TextureSample() {texture=/BRCosmetics/Weapons/FORT_Melee/Pickaxe_Rays_Female/Textures/T_Pickaxe_Rays_female_D}
  n2 = ComponentMask(None=@n1.RGBA) {b=true, g=true, r=true}
  Desaturation(None=@n2, Fraction=ScalarParameter() {parameterName=Desaturate})
  n3 = DotProduct(
   A=Transform(None=TextureSample() {samplerType=SAMPLERTYPE_Normal, texture=/BRCosmetics/Weapons/FORT_Melee/Pickaxe_Rays_Female/Textures/T_Pickaxe_Rays_female_N}.RGB)
   B=Transform(None=VectorParameter() {parameterName=Dot, defaultValue=<rgba 0 0 0 1>}.RGB) {transformSourceType=TRANSFORMSOURCE_Local}
  )
  n4 = ScalarParameter() {parameterName=dot_exp}
  n5 = LinearInterpolate(
   A=ScalarParameter() {defaultValue=1, group=Brightness, parameterName=MinBrightness}
   B=ScalarParameter() {defaultValue=1, group=Brightness, parameterName=MaxBrightness}
   Alpha=NamedRerouteUsage()
  )
  n6 = VectorParameter() {parameterName=EnergyColor, defaultValue=<rgba 0 0 0 1>}
  n7 = TextureSample(UVs=NamedRerouteUsage()) {samplerType=SAMPLERTYPE_LinearColor, texture=&/T_RaysFemale_Light_Blades_Flow}
  n8 = Subtract(A=Multiply(A=ComponentMask(None=@n7.RGB) {g=true, r=true}) {constB=2})
  n9 = ComponentMask(None=VectorParameter() {parameterName=FlowmapStr, defaultValue=<rgba 1 1 0 1>}.RGB) {g=true, r=true}
  n10 = VectorParameter() {parameterName=FlowmapUVscale, defaultValue=<rgba 1 1 0 0>}
  n11 = Multiply(A=NamedRerouteUsage(), B=ComponentMask(None=@n10.RGB) {g=true, r=true})
  n12 = AppendVector(A=@n10.B, B=@n10.B)
  n13 = ComponentMask(None=MaterialFunctionCall!(
    Diffuse=TextureObject() {texture=/Game/Effects/Fort_Effects/Textures/T_FX_SpringerPlasma}
    Normal Map
    "Flow Vector Map (see tooltip)"=Multiply(A=@n8, B=@n9)
    UV texture 2 offset=Constant2Vector() {g=0.5, r=0.5}
    "Time (see tooltip)"=Time()
    useMipLevel
    UVs=Panner(Coordinate=@n11, Speed=@n12)
    Texture Size
    Texture Mip Bias
   ) {materialFunction=/Engine/Functions/Engine_MaterialFunctions02/Texturing/Flowmaps.FlowMaps}.Diffuse) {b=true}
  n14 = ScalarParameter() {group=Energy, parameterName=Energy_Bias}
  n15 = NamedRerouteDeclaration(None=Saturate(None=Add(A=@n13, B=@n14))) {name=FlowNoise, nodeColor=<rgba 0.0078 1 0 1>}
  n16 = Power(Base=@n1.RGB) {constExponent=1.5}
  n17 = Add(A=Multiply(A=@n16, B=Multiply(A=@n6.RGB, B=@n15)), B=@n16)
  n18 = Add(A=Saturate(None=Power(Base=@n3, Exp=@n4)), B=Multiply(A=Multiply(A=@n5, B=@n17), B=NamedRerouteUsage()))
  n19 = MaterialFunctionCall() {materialFunction=&ObjectPivotPoint}
  n20 = WorldPosition()
  n21 = ScalarParameter() {group=InternalLight, parameterName=Internal_Light_Size}
  n22 = LinearInterpolate(
   A=ScalarParameter() {group=InternalLight, parameterName=internal_light_minbright}
   B=ScalarParameter() {group=InternalLight, parameterName=internal_light_maxbright}
   Alpha=Power(Base=SphereMask(A=@n19.Object Pivot Location, B=@n20.XYZ, Radius=@n21) {hardnessPercent=0})
  )
  n23 = Multiply(A=NamedRerouteUsage())
  n24 = SmoothStep(Min=@n23, Max=Add(A=@n23, B=ScalarParameter() {group=Reveal, parameterName=Reveal_W}), Value=NamedRerouteUsage())
  MaterialFunctionCall!(In=MakeMaterialAttributes!(
    BaseColor
    Metallic
    Specular
    Roughness
    Anisotropy
    EmissiveColor=Multiply(A=@n18, B=@n22)
    Opacity=@n24
    OpacityMask
    Normal
    Tangent
    WorldPositionOffset
    SubsurfaceColor
    ClearCoat
    ClearCoatRoughness
    AmbientOcclusion
    Refraction
    CustomizedUVs_0
    CustomizedUVs_1
    CustomizedUVs_2
    CustomizedUVs_3
    CustomizedUVs_4
    CustomizedUVs_5
    CustomizedUVs_6
    CustomizedUVs_7
    PixelDepthOffset
    ShadingModel
    Displacement
   )) {materialFunction=/Game/Effects/Niagara/MasterMaterials/Functions/MF_FX_EmissiveAdjustments_MA}
  Multiply(A=@n2, B=ComponentMask(None=@n1.RGBA) {a=true})
  Multiply(
   A=SphereMask(
    A=MaterialFunctionCall() {materialFunction=&ObjectPivotPoint}.Object Pivot Location
    B=WorldPosition().XYZ
    Radius=ScalarParameter() {parameterName=Darkness_Radius}
    Hardness=ScalarParameter() {parameterName=Darkness_H}
   ) {attenuationRadius=128, hardnessPercent=0}
   B=NamedRerouteUsage()
  )
  Multiply(A=ScalarParameter() {group=Reveal, parameterName=Reveal_HueShift}, B=@n24)
  n25 = TextureSample() {texture=/BRCosmetics/Weapons/FORT_Melee/Pickaxe_Rays_Female/Textures/T_Pickaxe_Rays_female_E}
  NamedRerouteDeclaration(None=Power(Base=@n25.G) {constExponent=1}) {name=AO, nodeColor=<rgba 0.7578 0 1 1>}
  n26 = NamedRerouteUsage()
  n27 = ScalarParameter() {group=Reveal, parameterName=Reveal_Noise_Contrib}
  n28 = TextureSample(UVs=Add(A=Multiply(A=@n26, B=@n27), B=NamedRerouteUsage())) {
   samplerSource=SSM_Clamp_WorldGroupSettings
   samplerType=SAMPLERTYPE_Masks
   texture=&/T_RaysF_Pickaxe_Timing
  }
  NamedRerouteDeclaration(None=@n28.G) {name=TimingMap, nodeColor=<rgba 1 0.8203 0 1>}
  NamedRerouteDeclaration(None=@n25.R) {name=EmissiveWeight, nodeColor=<rgba 0 0.0625 1 1>}
  n29 = VectorParameter() {parameterName=UV_Xfm, channelNames {a=Y_Scale, b=X_Scale, g=Y_Offs, r=X_Offs}, defaultValue=<rgba 0.472 0.672 57.6906 -105.6348>}
  n30 = Multiply(A=ComponentMask(None=PreSkinnedPosition()) {b=true, r=true}, B=Divide(B=AppendVector(A=@n29.X_Scale, B=@n29.X_Scale)))
  n31 = AppendVector(A=@n29.X_Offs, B=@n29.X_Offs)
  NamedRerouteDeclaration(None=VertexInterpolator(VS=Add(A=@n30, B=@n31)).PS) {name=FlatUV, nodeColor=<rgba 0 0.2969 1 1>}
  NamedRerouteDeclaration(None=@n25.G) {name=Rim, nodeColor=<rgba 0 1 0.9297 1>}
  n32 = Time()
  n33 = ScalarParameter() {parameterName=Debug_Reveal_Duration}
  n34 = Saturate(None=Fmod(A=Multiply(A=@n32, B=@n33), B=Constant() {r=1}))
  n35 = ScalarParameter() {group=Reveal, parameterName=Reveal}
  n36 = Power(
   Base=Multiply(A=StaticSwitchParameter(True=@n34, False=@n35) {parameterName=bUseDebugRevealTime}) {constB=1.1}
   Exp=ScalarParameter() {defaultValue=1, parameterName=RevealTime_Curve}
  )
  n37 = SphereMask(
   A=ComponentMask(None=NamedRerouteUsage()) {g=true}
   B=ScalarParameter() {group=Reveal Bias, parameterName=Reveal_Bias_Pos}
   Radius=ScalarParameter() {group=Reveal Bias, parameterName=Reveal_Bias_R}
  ) {hardnessPercent=0}
  n38 = ScalarParameter() {group=Reveal Bias, parameterName=RevealBias_Str}
  NamedRerouteDeclaration(None=Add(A=VertexInterpolator(VS=@n36).PS, B=Multiply(A=@n37, B=@n38))) {name=RevealTime, nodeColor=<rgba 0.4297 0 1 1>}
  OneMinus(None=@n37)
  PixelNormalWS()
 ]
}
