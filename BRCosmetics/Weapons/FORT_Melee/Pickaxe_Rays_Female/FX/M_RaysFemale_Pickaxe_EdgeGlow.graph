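# Additive, unlit, two-sided material for an edge glow on Niagara mesh particles of the Rays Female pickaxe.
# Panning noise warps an electronic gradient's UVs, driving particle-colored emissive and an opacity mask.
# Vertex color, inverted Fresnel, particle alpha and a dynamic-radius local-Z reveal shape opacity, with depth fade on higher quality levels.
# A parameterized vertex slide along a transformed vector tapers across U and scales out via IsInFrontEnd.
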Material /BRCosmetics/Weapons/FORT_Melee/Pickaxe_Rays_Female/FX/M_RaysFemale_Pickaxe_EdgeGlow
graph {
 properties {
  bEnableResponsiveAA=true
  bUsedWithNiagaraMeshParticles=true
  blendMode=BLEND_Additive
  pixelDepthOffsetMode=PDOM_Legacy
  shadingModel=MSM_Unlit
  twoSided=true
  usedShadingModels=Unlit
 }
 nodes [
  n1 = Panner(Coordinate=TextureCoordinate() {vTiling=0.5}) {speedX=-1, speedY=0.2}
  n2 = TextureSample(UVs=LinearInterpolate(
    A=TextureCoordinate() {uTiling=0.8}
    B=TextureCoordinate() {uTiling=0.35}
    Alpha=TextureSample(UVs=@n1) {samplerType=SAMPLERTYPE_Grayscale, texture=&LowResBlurredNoise}.G
   )) {texture=/Game/Weapons/WeaponSkins/Textures/ElectronicGradient_Flipped}
  MaterialFunctionCall(Emissive=Multiply(A=Multiply(A=@n2.G) {constB=3}, B=ParticleColor().RGB)) {materialFunction=&MF_FX_EmissiveAdjustments}
  n3 = ComponentMask(None=TextureCoordinate()) {r=true}
  n4 = ScalarParameter() {group=VS, parameterName=Slide_Edge}
  n5 = Normalize(VectorInput=Transform(None=VectorParameter() {group=VS, parameterName=TangentSlide, defaultValue=<rgba 0 0 0 1>}.RGB))
  Multiply(
   A=Multiply(A=OneMinus(None=@n3), B=Multiply(A=@n4, B=@n5))
   B=OneMinus(None=DynamicParameter() {defaultValue=<rgba 250 1 1 1>, paramNames [RevealRadius,IsInFrontEnd,Param3,Param4]}.IsInFrontEnd)
  )
  n6 = Power(Base=@n2.G) {constExponent=1.75}
  n7 = OneMinus(None=ComponentMask(None=TextureCoordinate() {uTiling=0.5}) {r=true})
  n8 = ScalarParameter() {parameterName=edge}
  n9 = Multiply(A=Add(A=Multiply(A=@n6) {constB=3}, B=Multiply(A=@n7, B=@n8) {constB=0.05}), B=VertexColor().R)
  n10 = DepthFade(Opacity=@n9) {fadeDistanceDefault=1}
  n11 = QualitySwitch(Default=@n10, Low=@n9, High=@n10, Medium=@n9, Epic=@n10)
  n12 = MaterialFunctionCall(
   Normal Vector=Multiply(A=VertexNormalWS(), B=TwoSidedSign())
   Invert Fresnel=StaticBool() {value=true}
   Power=ScalarParameter() {parameterName=Fresnel_Power}
  ) {materialFunction=&Fresnel_Function}
  n13 = MaterialFunctionCall() {materialFunction=&LocalPosition}
  n14 = Power(Base=SphereMask(
    A=ComponentMask(None=@n13.Local Position) {b=true}
    B=Constant()
    Radius=DynamicParameter() {defaultValue=<rgba 250 1 1 1>, paramNames [RevealRadius,IsInFrontEnd,Param3,Param4]}.RevealRadius
   ) {hardnessPercent=0})
  Multiply(A=ParticleColor().A, B=Multiply(A=Multiply(A=@n11, B=@n12.Result), B=VertexInterpolator(VS=@n14).PS))
 ]
}
