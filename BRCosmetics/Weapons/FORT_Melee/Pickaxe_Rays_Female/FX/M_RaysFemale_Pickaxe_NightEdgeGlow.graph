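# Additive, unlit, two-sided edge glow material for the female Rays pickaxe, built for Niagara mesh particles.
# Particle color is hue-shifted by panning noise in object-stabilized screen-space UVs, which also modulates opacity.
# An edge gradient, soft fill, vertex color, Fresnel and local-Z sphere-mask reveal shape the glow, with depth fade on higher quality levels.
# Dynamic parameters drive reveal radius, fill and a front-end switch that boosts opacity and cancels the tangent edge slide.
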Material /BRCosmetics/Weapons/FORT_Melee/Pickaxe_Rays_Female/FX/M_RaysFemale_Pickaxe_NightEdgeGlow
graph {
 properties {
  bEnableResponsiveAA=true
  bUsedWithNiagaraMeshParticles=true
  blendMode=BLEND_Additive
  pixelDepthOffsetMode=PDOM_Legacy
  shadingModel=MSM_Unlit
  twoSided=true
  usedShadingModels=Unlit
 }
 nodes [
  n1 = MaterialFunctionCall!(
   Position=MaterialFunctionCall() {materialFunction=&ObjectPivotPoint}.Object Pivot Location
   UVRadialScale=ScalarParameter() {parameterName=screenscale}
  ) {materialFunction=/Game/Weapons/WeaponSkins/MaterialFunction/ObjectStabilizedScreenSpaceUVs}
  n2 = TextureSample(UVs=Panner(Coordinate=Divide(A=@n1.StabilizedUVs) {constB=4}) {bFractionalPart=true, speedY=0.25}) {samplerSource=SSM_Wrap_WorldGroupSettings, samplerType=SAMPLERTYPE_Grayscale, texture=&LowResBlurredNoise}
  n3 = ScalarParameter() {parameterName=HueShiftMove}
  MaterialFunctionCall(Emissive=MaterialFunctionCall(Hue Shift Percentage=Multiply(A=@n2.R, B=@n3), Texture=ParticleColor().RGB) {materialFunction=&HueShift}.Result) {materialFunction=&MF_FX_EmissiveAdjustments}
  n4 = ComponentMask(None=TextureCoordinate()) {r=true}
  n5 = ScalarParameter() {group=VS, parameterName=Slide_Edge}
  n6 = Normalize(VectorInput=Transform(None=VectorParameter() {group=VS, parameterName=TangentSlide, defaultValue=<rgba 1 0 0 1>}.RGB))
  n7 = DynamicParameter() {defaultValue=<rgba 250 1 1 1>, paramNames [RevealRadius,IsInFrontEnd,SoftFill,Param4]}
  Multiply(A=Multiply(A=OneMinus(None=@n4), B=Multiply(A=@n5, B=@n6)), B=OneMinus(None=@n7.IsInFrontEnd))
  n8 = TextureSample(UVs=Panner(Coordinate=@n1.StabilizedUVs) {bFractionalPart=true, speedY=0.25}) {samplerSource=SSM_Wrap_WorldGroupSettings, samplerType=SAMPLERTYPE_Grayscale, texture=&LowResBlurredNoise}
  n9 = Add(
   A=Multiply(A=@n8.G, B=ScalarParameter() {parameterName=GalaxyOpacity_Boost})
   B=Multiply(A=ScalarParameter() {defaultValue=1, parameterName=FrontEndOpacityBoost}, B=@n7.IsInFrontEnd)
  )
  n10 = ParticleColor()
  n11 = LinearInterpolate(A=TextureCoordinate() {uTiling=1.2}, B=TextureCoordinate() {uTiling=2.5}, Alpha=NamedRerouteUsage())
  n12 = Power(
   Base=TextureSample(UVs=@n11) {texture=/Game/Weapons/WeaponSkins/Textures/ElectronicGradient_Flipped}.G
   Exp=ScalarParameter() {parameterName=edgepwr}
  ) {constExponent=1.75}
  n13 = ScalarParameter() {parameterName=edgestr}
  n14 = OneMinus(None=ComponentMask(None=TextureCoordinate()) {r=true})
  n15 = DynamicParameter() {defaultValue=<rgba 250 1 1 1>, paramNames [RevealRadius,IsInFrontEnd,SoftFill,Param4]}
  n16 = Multiply(A=Add(A=Multiply(A=@n12, B=@n13) {constB=3}, B=Multiply(A=@n14, B=@n15.SoftFill) {constB=0.05}), B=VertexColor().R)
  n17 = DepthFade(Opacity=@n16) {fadeDistanceDefault=1}
  n18 = Multiply(A=VertexNormalWS(), B=TwoSidedSign())
  n19 = StaticBool() {value=true}
  n20 = ScalarParameter() {parameterName=Fresnel_Power}
  n21 = SphereMask(
   A=ComponentMask(None=MaterialFunctionCall() {materialFunction=&LocalPosition}.Local Position) {b=true}
   B=Constant()
   Radius=DynamicParameter() {defaultValue=<rgba 250 1 1 1>, paramNames [RevealRadius,IsInFrontEnd,SoftFill,Param4]}.RevealRadius
  ) {hardnessPercent=0}
  n22 = Multiply(
   A=Multiply(
    A=QualitySwitch(Default=@n17, Low=@n16, High=@n17, Medium=@n16, Epic=@n17)
    B=MaterialFunctionCall(Normal Vector=@n18, Invert Fresnel=@n19, Power=@n20) {materialFunction=&Fresnel_Function}.Result
   )
   B=VertexInterpolator(VS=Power(Base=@n21)).PS
  )
  Multiply(A=Multiply(A=Saturate(None=@n9), B=Multiply(A=@n10.A, B=@n22)), B=NamedRerouteUsage())
  NamedRerouteDeclaration(None=ComponentMask(None=@n11) {r=true}) {name=EdgeGradient, nodeColor=<rgba 0.4531 0 1 1>}
  NamedRerouteDeclaration(None=@n8.G) {name=ScreenScaleSmoke, nodeColor=<rgba 1 0 0.1172 1>}
 ]
}
